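# Physics asset for the Calavera male medium body, setting up dynamic jacket bones with capsules on the pelvis, thighs and jacket chains.
# Ten left/right front, side and back chains hang from tiny master bodies, most locked angularly to the pelvis.
# Limited swing and twist joints link each chain's segments.
# Swing-locked cross-constraints tie some front and side chain tips together; thigh_r swings freely while thigh_l is locked.
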PhysicsAsset /BRCosmetics/Characters/Player/Male/Medium/Bodies/M_MED_Calavera/Meshes/M_MED_Calavera_Physics
bodies {
 Dyn_MSTER_Jacket_01_l [{shapeType=Capsule, length=0.5, radius=0.5, shapeName=Dyn_MSTER_Jacket_01_l_capsule}]
 Dyn_MSTER_Jacket_01_r [{shapeType=Capsule, length=0.5, radius=0.5, shapeName=Dyn_MSTER_Jacket_01_r_capsule}]
 Dyn_MSTR_Jacket_Mid_bk_l [{shapeType=Capsule, length=0.5, radius=0.5, shapeName=Dyn_MSTR_Jacket_Mid_bk_l_capsule}]
 Dyn_MSTR_Jacket_Mid_bk_r [{shapeType=Capsule, length=0.5, radius=0.5, shapeName=Dyn_MSTR_Jacket_Mid_bk_r_capsule}]
 Dyn_MSTR_Jacket_bk_l [{shapeType=Capsule, length=0.5, radius=0.5, shapeName=Dyn_MSTR_Jacket_bk_l_capsule}]
 Dyn_MSTR_Jacket_bk_r [{shapeType=Capsule, length=0.5, radius=0.5, shapeName=Dyn_MSTR_Jacket_bk_r_capsule}]
 Dyn_MSTR_Jacket_fr_01_l [{shapeType=Capsule, length=1, radius=1}]
 Dyn_MSTR_Jacket_fr_01_r [{shapeType=Capsule, length=1, radius=1}]
 Dyn_MSTR_Jacket_fr_mid_01_l [{shapeType=Capsule, length=0.5, radius=0.5, shapeName=Dyn_MSTR_Jacket_fr_mid_01_l_capsule}]
 Dyn_MSTR_Jacket_fr_mid_01_r [{shapeType=Capsule, length=0.5, radius=0.5, shapeName=Dyn_MSTR_Jacket_fr_mid_01_r_capsule}]
 dyn_Jacket_01_l [{shapeType=Capsule, length=3.3914, radius=3.6086, shapeName=dyn_Jacket_01_l_capsule, center=<2 -3.0708 0>}]
 dyn_Jacket_01_l1 [{shapeType=Capsule, length=3.4913, radius=3.5087, shapeName=dyn_Jacket_01_l1_capsule, center=<2 -4.2 0>}]
 dyn_Jacket_01_l2 [{shapeType=Capsule, length=3.4121, radius=3.5879, shapeName=dyn_Jacket_01_l2_capsule, center=<2 -3.8 0>}]
 dyn_Jacket_01_r [{shapeType=Capsule, length=3, radius=4, shapeName=dyn_Jacket_01_r_capsule, center=<2 -3.0708 0>}]
 dyn_Jacket_01_r1 [{shapeType=Capsule, length=9.0728, radius=4, shapeName=dyn_Jacket_01_r1_capsule, center=<2 -4.2 0>}]
 dyn_Jacket_01_r2 [{shapeType=Capsule, length=10.2583, radius=4, shapeName=dyn_Jacket_01_r2_capsule, center=<1.7077 -3.65 0>}]
 dyn_Jacket_Mid_bk_l1 [{shapeType=Capsule, length=3, radius=4, shapeName=dyn_Jacket_Mid_bk_l1_capsule, center=<2 -3.1133 0>}]
 dyn_Jacket_Mid_bk_l2 [{shapeType=Capsule, length=3, radius=4, shapeName=dyn_Jacket_Mid_bk_l2_capsule, center=<2 -3.1133 0>}]
 dyn_Jacket_Mid_bk_r1 [{shapeType=Capsule, length=3, radius=4, shapeName=dyn_Jacket_Mid_bk_r1_capsule, center=<2 -3.0708 0>}]
 dyn_Jacket_Mid_bk_r2 [{shapeType=Capsule, length=3, radius=4, shapeName=dyn_Jacket_Mid_bk_r2_capsule, center=<2 -3.0708 0>}]
 dyn_Jacket_bk_l1 [{shapeType=Capsule, length=2.1957, radius=4.8043, shapeName=dyn_Jacket_bk_l1_capsule, center=<2 -3.2022 0>}]
 dyn_Jacket_bk_l2 [{shapeType=Capsule, length=2.1957, radius=4.8043, shapeName=dyn_Jacket_bk_l2_capsule, center=<2 -3.2022 0>}]
 dyn_Jacket_bk_r1 [{shapeType=Capsule, length=2.1957, radius=4.8043, shapeName=dyn_Jacket_bk_r1_capsule, center=<2 -3.0708 0>}]
 dyn_Jacket_bk_r2 [{shapeType=Capsule, length=2.1957, radius=4.8043, shapeName=dyn_Jacket_bk_r2_capsule, center=<2 -3.0708 0>}]
 dyn_Jacket_fr_01_l [{shapeType=Capsule, length=3, radius=4, shapeName=dyn_Jacket_fr_01_l_capsule, center=<4.0728 -2.3819 0>}]
 dyn_Jacket_fr_01_l1 [{
   shapeType=Capsule
   length=7.9871
   radius=3.7253
   shapeName=dyn_Jacket_fr_01_l1_capsule
   center=<2.2 -3 0>
   rotation=<rot 0.0001 -0.0001 23.6852>
  }]
 dyn_Jacket_fr_01_l2 [{
   shapeType=Capsule
   length=7.4133
   radius=4.0658
   shapeName=dyn_Jacket_fr_01_l2_capsule
   center=<0.5 -3 0>
   rotation=<rot 0.0016 -0.0022 20.0739>
  }]
 dyn_Jacket_fr_01_r [{shapeType=Capsule, length=4.4905, radius=3.5842, center=<4.0303 -4 0>}]
 dyn_Jacket_fr_01_r1 [{shapeType=Capsule, length=6.87, radius=5.2551, center=<3.2121 -4 0>}]
 dyn_Jacket_fr_01_r2 [{shapeType=Capsule, length=9.7845, radius=5.0183, center=<2 -3.7002 0>}]
 dyn_Jacket_fr_mid_01_l [{shapeType=Capsule, length=6.5304, radius=3.744, shapeName=dyn_Jacket_fr_mid_01_l_capsule, center=<2 -3.14 0>}]
 dyn_Jacket_fr_mid_01_l1 [{
   shapeType=Capsule
   length=6.8519
   radius=3.4225
   shapeName=dyn_Jacket_fr_mid_01_l1_capsule
   center=<2 -4.0202 0>
   rotation=<rot 0.0002 -0.0001 -11.9808>
  }]
 dyn_Jacket_fr_mid_01_l2 [{
   shapeType=Capsule
   length=6.9481
   radius=3.3263
   shapeName=dyn_Jacket_fr_mid_01_l2_capsule
   center=<2 -3.65 0>
   rotation=<rot 0.0002 -0.0001 -11.9808>
  }]
 dyn_Jacket_fr_mid_01_r [{shapeType=Capsule, length=3, radius=4, shapeName=dyn_Jacket_fr_mid_01_r_capsule, center=<2 -3.14 0>}]
 dyn_Jacket_fr_mid_01_r1 [{shapeType=Capsule, length=9.0327, radius=4.4777, shapeName=dyn_Jacket_fr_mid_01_r1_capsule, center=<2 -4.5 0>}]
 dyn_Jacket_fr_mid_01_r2 [{shapeType=Capsule, length=9.4241, radius=4.5707, shapeName=dyn_Jacket_fr_mid_01_r2_capsule, center=<2 -3.9403 0>}]
 pelvis [{
   shapeType=Capsule
   length=6.5861
   radius=8.9692
   shapeName=pelvis_capsule
   center=<-0.8963 0.2916 0.2509>
   rotation=<rot 0 -8.8429 -172.4886>
  }]
 thigh_l [{shapeType=Capsule, length=29.4088, radius=7.0912, shapeName=thigh_l_capsule, center=<-14 0.0953 0>, rotation=<rot 90 0 0>}]
 thigh_r [{shapeType=Capsule, length=28.364, radius=8.136, shapeName=thigh_r_capsule, center=<14 1.0184 -0.4247>, rotation=<rot 90 0 0>}]
}
constraints [
 {
  bone1Name=Dyn_MSTR_Jacket_bk_r
  bone2Name=pelvis
  swing1LimitDegrees=45
  swing1Motion=Locked
  swing2LimitDegrees=45
  swing2Motion=Locked
  twistLimitDegrees=45
  twistMotion=Locked
 }
 {
  bone1Name=Dyn_MSTER_Jacket_01_r
  bone2Name=pelvis
  swing1LimitDegrees=45
  swing1Motion=Locked
  swing2LimitDegrees=45
  swing2Motion=Locked
  twistLimitDegrees=45
  twistMotion=Locked
 }
 {
  bone1Name=Dyn_MSTR_Jacket_Mid_bk_r
  bone2Name=pelvis
  swing1LimitDegrees=45
  swing1Motion=Locked
  swing2LimitDegrees=45
  swing2Motion=Locked
  twistLimitDegrees=45
  twistMotion=Locked
 }
 {
  bone1Name=Dyn_MSTR_Jacket_bk_l
  bone2Name=pelvis
  swing1LimitDegrees=45
  swing1Motion=Locked
  swing2LimitDegrees=45
  swing2Motion=Locked
  twistLimitDegrees=45
  twistMotion=Locked
 }
 {
  bone1Name=Dyn_MSTR_Jacket_Mid_bk_l
  bone2Name=pelvis
  swing1LimitDegrees=45
  swing1Motion=Locked
  swing2LimitDegrees=45
  swing2Motion=Locked
  twistLimitDegrees=45
  twistMotion=Locked
 }
 {
  bone1Name=Dyn_MSTER_Jacket_01_l
  bone2Name=pelvis
  swing1LimitDegrees=45
  swing1Motion=Locked
  swing2LimitDegrees=45
  swing2Motion=Locked
  twistLimitDegrees=45
  twistMotion=Locked
 }
 {
  bone1Name=thigh_l
  bone2Name=pelvis
  swing1LimitDegrees=45
  swing1Motion=Locked
  swing2LimitDegrees=45
  swing2Motion=Locked
  twistLimitDegrees=45
  twistMotion=Locked
 }
 {bone1Name=dyn_Jacket_fr_mid_01_r, bone2Name=Dyn_MSTR_Jacket_fr_mid_01_r, swing1LimitDegrees=35, swing2LimitDegrees=10, twistLimitDegrees=10}
 {bone1Name=dyn_Jacket_fr_mid_01_r1, bone2Name=dyn_Jacket_fr_mid_01_r, swing1LimitDegrees=45, swing2LimitDegrees=10, twistLimitDegrees=10}
 {bone1Name=dyn_Jacket_fr_mid_01_r2, bone2Name=dyn_Jacket_fr_mid_01_r1, swing1LimitDegrees=45, swing2LimitDegrees=10, twistLimitDegrees=10}
 {bone1Name=dyn_Jacket_fr_mid_01_l, bone2Name=Dyn_MSTR_Jacket_fr_mid_01_l, swing1LimitDegrees=30, swing2LimitDegrees=10, twistLimitDegrees=20}
 {bone1Name=dyn_Jacket_fr_mid_01_l1, bone2Name=dyn_Jacket_fr_mid_01_l, swing1LimitDegrees=20, swing2LimitDegrees=5, twistLimitDegrees=10}
 {bone1Name=dyn_Jacket_fr_mid_01_l2, bone2Name=dyn_Jacket_fr_mid_01_l1, swing1LimitDegrees=20, swing2LimitDegrees=5, twistLimitDegrees=10}
 {bone1Name=dyn_Jacket_fr_01_l1, bone2Name=dyn_Jacket_fr_01_l, swing1LimitDegrees=40, swing2LimitDegrees=5, twistLimitDegrees=10}
 {bone1Name=dyn_Jacket_fr_01_l2, bone2Name=dyn_Jacket_fr_01_l1, swing1LimitDegrees=20, swing2LimitDegrees=5, twistLimitDegrees=10}
 {bone1Name=dyn_Jacket_bk_r1, bone2Name=Dyn_MSTR_Jacket_bk_r, swing1LimitDegrees=45, swing2LimitDegrees=30, twistLimitDegrees=10}
 {bone1Name=dyn_Jacket_bk_r2, bone2Name=dyn_Jacket_bk_r1, swing1LimitDegrees=45, swing2LimitDegrees=30, twistLimitDegrees=10}
 {bone1Name=dyn_Jacket_bk_l1, bone2Name=Dyn_MSTR_Jacket_bk_l, swing1LimitDegrees=45, swing2LimitDegrees=30, twistLimitDegrees=10}
 {bone1Name=dyn_Jacket_bk_l2, bone2Name=dyn_Jacket_bk_l1, swing1LimitDegrees=45, swing2LimitDegrees=30, twistLimitDegrees=10}
 {bone1Name=dyn_Jacket_Mid_bk_r1, bone2Name=Dyn_MSTR_Jacket_Mid_bk_r, swing1LimitDegrees=45, swing2LimitDegrees=30, twistLimitDegrees=10}
 {bone1Name=dyn_Jacket_Mid_bk_r2, bone2Name=dyn_Jacket_Mid_bk_r1, swing1LimitDegrees=45, swing2LimitDegrees=30, twistLimitDegrees=10}
 {bone1Name=dyn_Jacket_Mid_bk_l1, bone2Name=Dyn_MSTR_Jacket_Mid_bk_l, swing1LimitDegrees=45, swing2LimitDegrees=30, twistLimitDegrees=10}
 {bone1Name=dyn_Jacket_Mid_bk_l2, bone2Name=dyn_Jacket_Mid_bk_l1, swing1LimitDegrees=45, swing2LimitDegrees=30, twistLimitDegrees=10}
 {bone1Name=dyn_Jacket_01_r, bone2Name=Dyn_MSTER_Jacket_01_r, swing1LimitDegrees=45, swing2LimitDegrees=30, twistLimitDegrees=10}
 {bone1Name=dyn_Jacket_01_r1, bone2Name=dyn_Jacket_01_r, swing1LimitDegrees=45, swing2LimitDegrees=30, twistLimitDegrees=10}
 {bone1Name=dyn_Jacket_01_l, bone2Name=Dyn_MSTER_Jacket_01_l, swing1LimitDegrees=45, swing2LimitDegrees=30, twistLimitDegrees=10}
 {bone1Name=dyn_Jacket_01_l1, bone2Name=dyn_Jacket_01_l, swing1LimitDegrees=45, swing2LimitDegrees=30, twistLimitDegrees=10}
 {bone1Name=dyn_Jacket_01_l2, bone2Name=dyn_Jacket_01_l1, swing1LimitDegrees=45, swing2LimitDegrees=30, twistLimitDegrees=10}
 {bone1Name=dyn_Jacket_01_r2, bone2Name=dyn_Jacket_01_r1, swing1LimitDegrees=30, swing2LimitDegrees=10, twistLimitDegrees=10}
 {bone1Name=dyn_Jacket_fr_01_r2, bone2Name=dyn_Jacket_fr_01_r1, swing1LimitDegrees=25, swing2LimitDegrees=10, twistLimitDegrees=2}
 {bone1Name=dyn_Jacket_fr_01_r1, bone2Name=dyn_Jacket_fr_01_r, swing1LimitDegrees=35, swing2LimitDegrees=10, twistLimitDegrees=2}
 {bone1Name=dyn_Jacket_fr_01_r, bone2Name=Dyn_MSTR_Jacket_fr_01_r, swing1LimitDegrees=35, swing2LimitDegrees=10, twistLimitDegrees=2}
 {
  bone1Name=Dyn_MSTR_Jacket_fr_01_r
  bone2Name=pelvis
  swing1LimitDegrees=45
  swing1Motion=Locked
  swing2LimitDegrees=45
  swing2Motion=Locked
  twistLimitDegrees=45
  twistMotion=Locked
 }
 {
  bone1Name=thigh_r
  bone2Name=pelvis
  swing1LimitDegrees=45
  swing1Motion=Free
  swing2LimitDegrees=45
  swing2Motion=Free
  twistLimitDegrees=45
  twistMotion=Free
 }
 {
  bone1Name=dyn_Jacket_fr_mid_01_r2
  bone2Name=dyn_Jacket_fr_01_r2
  swing1LimitDegrees=45
  swing1Motion=Locked
  swing2LimitDegrees=45
  swing2Motion=Locked
  twistLimitDegrees=10
 }
 {bone1Name=dyn_Jacket_fr_01_l, bone2Name=Dyn_MSTR_Jacket_fr_01_l, swing1LimitDegrees=40, swing2LimitDegrees=5, twistLimitDegrees=2}
 {
  bone1Name=Dyn_MSTR_Jacket_fr_01_l
  bone2Name=pelvis
  swing1LimitDegrees=45
  swing1Motion=Locked
  swing2LimitDegrees=45
  swing2Motion=Locked
  twistLimitDegrees=45
  twistMotion=Locked
 }
 {
  bone1Name=dyn_Jacket_01_r2
  bone2Name=dyn_Jacket_fr_mid_01_r2
  swing1LimitDegrees=45
  swing1Motion=Locked
  swing2LimitDegrees=45
  swing2Motion=Locked
  twistLimitDegrees=1
 }
 {
  bone1Name=dyn_Jacket_01_l2
  bone2Name=dyn_Jacket_fr_mid_01_l2
  swing1LimitDegrees=45
  swing1Motion=Locked
  swing2LimitDegrees=45
  swing2Motion=Locked
  twistLimitDegrees=1
 }
]
