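# Physics asset for the M_MED_Calavera male player body's dangling clothing: tie, shirt collar and jacket flaps.
# Capsules cover the Dyn_ cloth bones, plus spine, clavicles and thighs as body volumes.
# A four-link tie, the collar and upper-right jacket chain from spine_05; other jacket flaps hang from the clavicles.
# Torso and Dyn_MSTR anchor joints leave angular motion free, while cloth joints have limited swing and mostly near-locked twist.
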PhysicsAsset /BRCosmetics/Characters/Player/Male/Medium/Bodies/M_MED_Calavera/Meshes/M_MED_Calavera_Tie_Physics
bodies {
 Dyn_Jacket_Lower_l1 [{shapeType=Capsule, length=10, radius=2.3018, shapeName=Dyn_Jacket_Lower_l1_capsule, center=<-1.5454 -0.2755 0.2602>, rotation=<rot 0 -90 0>}]
 Dyn_Jacket_Lower_r1 [{shapeType=Capsule, length=12, radius=2.1123, shapeName=Dyn_Jacket_Lower_r1_capsule, center=<-1.3679 -2.8166 -1.8357>, rotation=<rot 0 -90 0>}]
 Dyn_Jacket_Upper_l1 [{
   shapeType=Capsule
   length=3.6778
   radius=1.7885
   shapeName=Dyn_Jacket_Upper_l1_capsule
   center=<-1.2417 -2.4153 -1.0371>
   rotation=<rot 21.031 -90 -3.1663>
  }]
 Dyn_Jacket_Upper_r1 [{
   shapeType=Capsule
   length=8.6797
   radius=1.6775
   shapeName=Dyn_Jacket_Upper_r1_capsule
   center=<-0.5949 -0.6026 -0.4344>
   rotation=<rot 0.858 -90 1.708>
  }]
 Dyn_Jacket_Upper_r2 [{
   shapeType=Capsule
   length=2.9381
   radius=1
   shapeName=Dyn_Jacket_Upper_r2_capsule
   center=<0.1149 -1.7345 0.4834>
   rotation=<rot 17.9957 -90 1.0365>
  }]
 Dyn_MSTR_Jacket_Lower_l [{shapeType=Capsule, length=0.505, radius=0.505, shapeName=Dyn_MSTR_Jacket_Lower_l_capsule}]
 Dyn_MSTR_Jacket_Lower_r [{shapeType=Capsule, length=0.505, radius=0.505, shapeName=Dyn_MSTR_Jacket_Lower_r_capsule}]
 Dyn_MSTR_Jacket_Upper_l [{shapeType=Capsule, length=0.505, radius=0.505, shapeName=Dyn_MSTR_Jacket_Upper_l_capsule}]
 Dyn_MSTR_ShirtCollar [{shapeType=Capsule, length=0.5, radius=0.5, shapeName=Dyn_MSTR_ShirtCollar_capsule, rotation=<rot 0 -90 0>}]
 Dyn_MSTR_Tie [{shapeType=Capsule, length=0.505, radius=0.505, shapeName=Dyn_MSTR_Tie_capsule}]
 Dyn_ShirtCollar1 [{shapeType=Capsule, length=3.2498, radius=2, shapeName=Dyn_ShirtCollar1_capsule, center=<-0.4712 -1.8 -0.9871>}]
 Dyn_Tie1 [{shapeType=Capsule, length=5, radius=2.2026, shapeName=Dyn_Tie1_capsule, center=<3.2149 -1.7176 0>, rotation=<rot 90 0 -8.4373>}]
 Dyn_Tie2 [{shapeType=Capsule, length=8, radius=2, shapeName=Dyn_Tie2_capsule, center=<3.7057 -0.854 0>, rotation=<rot 90 0 0>}]
 Dyn_Tie3 [{shapeType=Capsule, length=10, radius=2.2217, shapeName=Dyn_Tie3_capsule, center=<2 -1.1488 0.615>, rotation=<rot 90 0 0>}]
 Dyn_Tie4 [{shapeType=Capsule, length=10, radius=3, shapeName=Dyn_Tie4_capsule, center=<3.3088 -1.6 1.0126>, rotation=<rot 90 0 0>}]
 clavicle_l [{shapeType=Capsule, length=12, radius=7, shapeName=clavicle_l_capsule, center=<7.2858 0.7866 1.2519>, rotation=<rot 81.5625 -0.0012 0.0012>}]
 clavicle_r [{
   shapeType=Capsule
   length=12
   radius=7
   shapeName=clavicle_r_capsule
   center=<-7.1658 -0.7866 -1.418>
   rotation=<rot 81.5625 -0.0012 -0.0012>
  }]
 spine_01 [{shapeType=Capsule, length=12.6349, radius=10.2369, shapeName=spine_01_capsule, center=<5.6216 1.3547 0.1461>}]
 spine_02 [{shapeType=Capsule, length=10.6741, radius=10.5542, center=<-10.499 0 0>}]
 spine_03 [{shapeType=Capsule, length=9.6242, radius=13.3758, shapeName=spine_03_capsule, center=<-0.0334 -1.4919 0>, rotation=<rot 90 180 175.0001>}]
 spine_04 [{shapeType=Capsule, length=5.6555, radius=10.4747, shapeName=spine_04_capsule, center=<2.8728 1.1162 0>}]
 spine_05 [{shapeType=Capsule, length=12.6496, radius=11.0604, shapeName=spine_05_capsule, center=<0 -0.5345 0>}]
 thigh_l [{
   shapeType=Capsule
   length=24.1932
   radius=6.673
   shapeName=thigh_l_capsule
   center=<-23.2603 -2.4009 0.4329>
   rotation=<rot 0 71.4442 90.4246>
  }]
 thigh_r [{
   shapeType=Capsule
   length=24.1932
   radius=6.673
   shapeName=thigh_r_capsule
   center=<23.2603 2.4009 -0.4329>
   rotation=<rot 0 71.4442 -89.5754>
  }]
}
constraints [
 {
  bone1Name=spine_04
  bone2Name=spine_03
  swing1LimitDegrees=45
  swing1Motion=Free
  swing2LimitDegrees=45
  swing2Motion=Free
  twistLimitDegrees=45
  twistMotion=Free
 }
 {
  bone1Name=spine_05
  bone2Name=spine_04
  swing1LimitDegrees=45
  swing1Motion=Free
  swing2LimitDegrees=45
  swing2Motion=Free
  twistLimitDegrees=45
  twistMotion=Free
 }
 {
  bone1Name=Dyn_MSTR_Tie
  bone2Name=spine_05
  swing1LimitDegrees=45
  swing1Motion=Free
  swing2LimitDegrees=45
  swing2Motion=Free
  twistLimitDegrees=45
  twistMotion=Free
 }
 {bone1Name=Dyn_Tie1, bone2Name=Dyn_MSTR_Tie, swing1LimitDegrees=45, swing2LimitDegrees=45, twistLimitDegrees=4}
 {bone1Name=Dyn_Tie2, bone2Name=Dyn_Tie1, swing1LimitDegrees=30, swing2LimitDegrees=40, twistLimitDegrees=1}
 {bone1Name=Dyn_Tie3, bone2Name=Dyn_Tie2, swing1LimitDegrees=30, swing2LimitDegrees=30, twistLimitDegrees=1}
 {bone1Name=Dyn_Tie4, bone2Name=Dyn_Tie3, swing1LimitDegrees=30, swing2LimitDegrees=30, twistLimitDegrees=1}
 {
  bone1Name=Dyn_MSTR_ShirtCollar
  bone2Name=spine_05
  swing1LimitDegrees=45
  swing1Motion=Free
  swing2LimitDegrees=45
  swing2Motion=Free
  twistLimitDegrees=45
  twistMotion=Free
 }
 {bone1Name=Dyn_ShirtCollar1, bone2Name=Dyn_MSTR_ShirtCollar, swing1LimitDegrees=45, swing2LimitDegrees=1, twistLimitDegrees=1}
 {bone1Name=Dyn_Jacket_Upper_r1, bone2Name=spine_05, swing1LimitDegrees=40, swing2LimitDegrees=20, twistLimitDegrees=10}
 {bone1Name=Dyn_Jacket_Upper_r2, bone2Name=Dyn_Jacket_Upper_r1, swing1LimitDegrees=20, swing2LimitDegrees=1, twistLimitDegrees=1}
 {
  bone1Name=clavicle_r
  bone2Name=spine_05
  swing1LimitDegrees=45
  swing1Motion=Free
  swing2LimitDegrees=45
  swing2Motion=Free
  twistLimitDegrees=45
  twistMotion=Free
 }
 {
  bone1Name=Dyn_MSTR_Jacket_Lower_r
  bone2Name=clavicle_r
  swing1LimitDegrees=45
  swing1Motion=Free
  swing2LimitDegrees=45
  swing2Motion=Free
  twistLimitDegrees=45
  twistMotion=Free
 }
 {bone1Name=Dyn_Jacket_Lower_r1, bone2Name=Dyn_MSTR_Jacket_Lower_r, swing1LimitDegrees=40, swing2LimitDegrees=1, twistLimitDegrees=1}
 {
  bone1Name=Dyn_MSTR_Jacket_Lower_l
  bone2Name=clavicle_l
  swing1LimitDegrees=45
  swing1Motion=Free
  swing2LimitDegrees=45
  swing2Motion=Free
  twistLimitDegrees=45
  twistMotion=Free
 }
 {
  bone1Name=Dyn_MSTR_Jacket_Upper_l
  bone2Name=clavicle_l
  swing1LimitDegrees=45
  swing1Motion=Free
  swing2LimitDegrees=45
  swing2Motion=Free
  twistLimitDegrees=45
  twistMotion=Free
 }
 {bone1Name=Dyn_Jacket_Lower_l1, bone2Name=Dyn_MSTR_Jacket_Lower_l, swing1LimitDegrees=45, swing2LimitDegrees=1, twistLimitDegrees=1}
 {bone1Name=Dyn_Jacket_Upper_l1, bone2Name=Dyn_MSTR_Jacket_Upper_l, swing1LimitDegrees=45, swing2LimitDegrees=1, twistLimitDegrees=1}
 {
  bone1Name=clavicle_l
  bone2Name=spine_05
  swing1LimitDegrees=45
  swing1Motion=Free
  swing2LimitDegrees=45
  swing2Motion=Free
  twistLimitDegrees=45
  twistMotion=Free
 }
 {
  bone1Name=spine_03
  bone2Name=spine_02
  swing1LimitDegrees=45
  swing1Motion=Free
  swing2LimitDegrees=45
  swing2Motion=Free
  twistLimitDegrees=45
  twistMotion=Free
 }
 {
  bone1Name=spine_02
  bone2Name=spine_01
  swing1LimitDegrees=45
  swing1Motion=Free
  swing2LimitDegrees=45
  swing2Motion=Free
  twistLimitDegrees=45
  twistMotion=Free
 }
]
