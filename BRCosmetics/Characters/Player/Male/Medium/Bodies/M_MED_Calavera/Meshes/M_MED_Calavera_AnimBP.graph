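# Post-process-style AnimBP for the Calavera medium male body, adding secondary motion to its clothing and tie.
# The parent's copied pose runs through a Control Rig (emote/DBNO aware), spine-based tie and pelvis-based body RigidBody sims, then an attach_backpack offset.
# Jacket, shirt and tie correctives, constraints and a third RigidBody are not wired to output.
# EventGraph computes Gravity Override and flags RainbowSplash backpacks; neither value is read here.
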AnimBlueprint /BRCosmetics/Characters/Player/Male/Medium/Bodies/M_MED_Calavera/Meshes/M_MED_Calavera_AnimBP
target_skeleton=&/M_MED_Calavera_Skeleton
graphs [
 AnimGraph {nodes [
   "Constraint - Dyn_Tie4 (U)"!(ComponentPose, bAlphaBoolEnabled:bool=True, ConstraintWeights_0=1) -> (Pose) {
    alphaInputType=Bool
    alphaBoolBlend {blendOption=Linear}
    boneToModify {boneName=Dyn_Tie4}
    constraintSetup [{transformType=LookAt, targetBone {boneName=spine_02}}]
    constraintWeights [1]
   }
   n1 = Get Standing_Moving() -> (Standing_Moving:bool)
   n2 = "Transform (Modify) Bone - Bone: Dyn_MSTR_Jacket_Mid_bk_l"!(ComponentPose, bAlphaBoolEnabled:bool=@n1, Translation="0,-1.3,0", Rotation="0,0,0", Scale="1,1,1") -> (Pose) {
    alphaInputType=Bool
    translationMode=BMM_Additive
    translationSpace=BCS_BoneSpace
    alphaBoolBlend {blendInTime=0.5, blendOption=Linear, blendOutTime=0.5}
    boneToModify {boneName=Dyn_MSTR_Jacket_Mid_bk_l}
    translation {y=-1.3}
   }
   n3 = "Transform (Modify) Bone - Bone: Dyn_MSTR_Jacket_bk_l"!(ComponentPose=@n2, bAlphaBoolEnabled:bool=@n1, Translation="0,-1.3,0", Rotation="0,0,0", Scale="1,1,1") -> (Pose) {
    alphaInputType=Bool
    translationMode=BMM_Additive
    translationSpace=BCS_BoneSpace
    alphaBoolBlend {blendInTime=0.5, blendOption=Linear, blendOutTime=0.5}
    boneToModify {boneName=Dyn_MSTR_Jacket_bk_l}
    translation {y=-1.3}
   }
   "Transform (Modify) Bone - Bone: Dyn_MSTR_Jacket_bk_r"!(ComponentPose=@n3, bAlphaBoolEnabled:bool=@n1, Translation="0,-1,0", Rotation="0,0,0", Scale="1,1,1") -> (Pose) {
    alphaInputType=Bool
    translationMode=BMM_Additive
    translationSpace=BCS_BoneSpace
    alphaBoolBlend {blendInTime=0.5, blendOption=Linear, blendOutTime=0.5}
    boneToModify {boneName=Dyn_MSTR_Jacket_bk_r}
    translation {y=-1}
   }
   "Transform (Modify) Bone - Bone: Dyn_MSTR_Tie"!(
    ComponentPose
    bAlphaBoolEnabled:bool=Get Is Rigid Body Disabled() -> (Is Rigid Body Disabled:bool)
    Translation="0,0,0"
    Rotation="0,-10,0"
    Scale="1,1,1"
   ) -> (Pose) {
    alphaInputType=Bool
    rotationMode=BMM_Additive
    rotationSpace=BCS_BoneSpace
    scaleMode=BMM_Additive
    scaleSpace=BCS_BoneSpace
    translationMode=BMM_Replace
    translationSpace=BCS_BoneSpace
    alphaBoolBlend {blendOption=Linear}
    boneToModify {boneName=Dyn_MSTR_Tie}
    rotation {yaw=-10}
   }
   n4 = "Constraint - dyn_Shirt_l_02 (UT)"!(ComponentPose, Alpha=1, ConstraintWeights_0=1, ConstraintWeights_1:float=0) -> (Pose) {
    alphaBoolBlend {blendOption=Linear}
    boneToModify {boneName=dyn_Shirt_l_02}
    constraintSetup [{transformType=LookAt, targetBone {boneName=thigh_l}}, {offsetOption=None, transformType=Translation, targetBone {boneName=None}}]
    constraintWeights [1,1]
   }
   n5 = Get Left Leg_crouch() -> (Left Leg_crouch:double)
   n6 = "Transform (Modify) Bone - Bone: Dyn_MSTR_Shirt_fl"(ComponentPose=@n4, Alpha=@n5, Translation="0.433,-2.086,2.211", Rotation="5.351,-92.178,38.397", Scale="0.7,0.7,0.7") {
    rotationMode=BMM_Additive
    rotationSpace=BCS_BoneSpace
    scaleMode=BMM_Additive
    scaleSpace=BCS_BoneSpace
    translationMode=BMM_Additive
    translationSpace=BCS_BoneSpace
    alphaBoolBlend {blendOption=Linear}
    alphaScaleBiasClamp {bInterpResult=true, interpSpeedIncreasing=100}
    boneToModify {boneName=Dyn_MSTR_Shirt_fl}
    rotation=<rot 5.351 -92.1781 38.397>
    scale=<0.7 0.7 0.7>
    translation=<0.433 -2.086 2.211>
   }
   n7 = "Transform (Modify) Bone - Bone: Dyn_MSTR_Jacket_fr_01_l"(ComponentPose=@n6, Alpha=@n5, Translation="-2,-1,0", Rotation="24.689,-98.290,20.761", Scale="1,1,1") {
    rotationMode=BMM_Additive
    rotationSpace=BCS_BoneSpace
    scaleMode=BMM_Additive
    scaleSpace=BCS_BoneSpace
    translationMode=BMM_Additive
    translationSpace=BCS_BoneSpace
    alphaBoolBlend {blendOption=Linear}
    alphaScaleBiasClamp {bInterpResult=true, interpSpeedIncreasing=100}
    boneToModify {boneName=Dyn_MSTR_Jacket_fr_01_l}
    rotation=<rot 24.689 -98.2901 20.761>
    translation=<-2 -1>
   }
   n8 = "Transform (Modify) Bone - Bone: Dyn_MSTR_Jacket_fr_mid_01_l"(ComponentPose=@n7, Alpha=@n5, Translation="-2,-2,0", Rotation="15.622,-100,1.126", Scale="1,1,1") {
    rotationMode=BMM_Additive
    rotationSpace=BCS_BoneSpace
    scaleMode=BMM_Additive
    scaleSpace=BCS_BoneSpace
    translationMode=BMM_Additive
    translationSpace=BCS_BoneSpace
    alphaBoolBlend {blendOption=Linear}
    alphaScaleBiasClamp {bInterpResult=true}
    boneToModify {boneName=Dyn_MSTR_Jacket_fr_mid_01_l}
    rotation=<rot 15.622 -100 1.126>
    translation=<-2 -2>
   }
   n9 = "Transform (Modify) Bone - Bone: Dyn_MSTER_Jacket_01_l"(ComponentPose=@n8, Alpha=@n5, Translation="-2,-2,0", Rotation="44.161,-35.872,-33.759", Scale="1,1,1") {
    rotationMode=BMM_Additive
    rotationSpace=BCS_BoneSpace
    scaleMode=BMM_Additive
    scaleSpace=BCS_BoneSpace
    translationMode=BMM_Additive
    translationSpace=BCS_BoneSpace
    alphaBoolBlend {blendOption=Linear}
    alphaScaleBiasClamp {bInterpResult=true}
    boneToModify {boneName=Dyn_MSTER_Jacket_01_l}
    rotation=<rot 44.161 -35.872 -33.759>
    translation=<-2 -2>
   }
   n10 = Get Left Leg() -> (Left Leg:double)
   n11 = "Transform (Modify) Bone - Bone: Dyn_MSTR_Jacket_fr_01_l"(ComponentPose=@n9, Alpha=@n10, Translation="-2,0,0", Rotation="-0.000,-11.350,-0.034", Scale="1,1,1") {
    rotationMode=BMM_Additive
    rotationSpace=BCS_BoneSpace
    scaleMode=BMM_Additive
    scaleSpace=BCS_BoneSpace
    translationMode=BMM_Additive
    translationSpace=BCS_BoneSpace
    alphaBoolBlend {blendOption=Linear}
    alphaScaleBiasClamp {bInterpResult=true}
    boneToModify {boneName=Dyn_MSTR_Jacket_fr_01_l}
    rotation {roll=-0.034, yaw=-11.35}
    translation {x=-2}
   }
   n12 = "Transform (Modify) Bone - Bone: Dyn_MSTR_Jacket_fr_mid_01_l"(ComponentPose=@n11, Alpha=@n10, Translation="0,0,0", Rotation="-0.413,22.233,-0.155", Scale="1,1,1") {
    rotationMode=BMM_Additive
    rotationSpace=BCS_BoneSpace
    scaleMode=BMM_Additive
    scaleSpace=BCS_BoneSpace
    translationMode=BMM_Additive
    translationSpace=BCS_BoneSpace
    alphaBoolBlend {blendOption=Linear}
    alphaScaleBiasClamp {bInterpResult=true}
    boneToModify {boneName=Dyn_MSTR_Jacket_fr_mid_01_l}
    rotation=<rot -0.413 22.233 -0.155>
   }
   n13 = "Transform (Modify) Bone - Bone: Dyn_MSTER_Jacket_01_l"(ComponentPose=@n12, Alpha=@n10, Translation="0,0,0", Rotation="0.012,3.094,-0.002", Scale="1,1,1") {
    rotationMode=BMM_Additive
    rotationSpace=BCS_BoneSpace
    scaleMode=BMM_Additive
    scaleSpace=BCS_BoneSpace
    translationMode=BMM_Additive
    translationSpace=BCS_BoneSpace
    alphaBoolBlend {blendOption=Linear}
    alphaScaleBiasClamp {bInterpResult=true}
    boneToModify {boneName=Dyn_MSTER_Jacket_01_l}
    rotation=<rot 0.012 3.094 -0.002>
   }
   n14 = Get Right Leg_Crouch() -> (Right Leg_Crouch:double)
   n15 = "Transform (Modify) Bone - Bone: Dyn_MSTR_Jacket_fr_01_r"(ComponentPose=@n13, Alpha=@n14, Translation="-2,0,2", Rotation="30,-90.000,0.000", Scale="1,1,1") {
    rotationMode=BMM_Additive
    rotationSpace=BCS_BoneSpace
    scaleMode=BMM_Additive
    scaleSpace=BCS_BoneSpace
    translationMode=BMM_Additive
    translationSpace=BCS_BoneSpace
    alphaBoolBlend {blendOption=Linear}
    alphaScaleBiasClamp {bInterpResult=true, interpSpeedDecreasing=100, interpSpeedIncreasing=100}
    boneToModify {boneName=Dyn_MSTR_Jacket_fr_01_r}
    rotation {pitch=30, yaw=-90}
    translation {x=-2, z=2}
   }
   n16 = "Transform (Modify) Bone - Bone: Dyn_MSTR_Jacket_fr_mid_01_r"(ComponentPose=@n15, Alpha=@n14, Translation="-2,0,2", Rotation="0,0,0", Scale="1,1,1") {
    rotationMode=BMM_Additive
    rotationSpace=BCS_BoneSpace
    scaleMode=BMM_Additive
    scaleSpace=BCS_BoneSpace
    translationMode=BMM_Additive
    translationSpace=BCS_BoneSpace
    alphaBoolBlend {blendOption=Linear}
    alphaScaleBiasClamp {bInterpResult=true, interpSpeedDecreasing=100, interpSpeedIncreasing=100}
    boneToModify {boneName=Dyn_MSTR_Jacket_fr_mid_01_r}
    translation {x=-2, z=2}
   }
   n17 = "Transform (Modify) Bone - Bone: Dyn_MSTR_Jacket_Upper_l"(ComponentPose=@n16, Alpha=Get ADS_UP_L_Pistol() -> (ADS_UP_L_Pistol:double), Translation="-1.5,0,0", Rotation="0,0,0", Scale="1,1,1") {
    rotationMode=BMM_Additive
    rotationSpace=BCS_BoneSpace
    scaleMode=BMM_Replace
    scaleSpace=BCS_BoneSpace
    translationMode=BMM_Additive
    translationSpace=BCS_BoneSpace
    alphaBoolBlend {blendOption=Linear}
    alphaScaleBiasClamp {bInterpResult=true}
    boneToModify {boneName=Dyn_MSTR_Jacket_Upper_l}
    translation {x=-1.5}
   }
   n18 = "Transform (Modify) Bone - Bone: Dyn_MSTR_Jacket_Upper_r"(
    ComponentPose=@n17
    Alpha=Get ADS_UP_R_Pistol() -> (ADS_UP_R_Pistol:double)
    Translation="-2.610,-2.637,0.955"
    Rotation="0,0,-10"
    Scale="0.8,0.8,0.8"
   ) {
    rotationMode=BMM_Additive
    rotationSpace=BCS_BoneSpace
    scaleMode=BMM_Replace
    scaleSpace=BCS_BoneSpace
    translationMode=BMM_Additive
    translationSpace=BCS_BoneSpace
    alphaBoolBlend {blendOption=Linear}
    alphaScaleBiasClamp {bInterpResult=true}
    boneToModify {boneName=Dyn_MSTR_Jacket_Upper_r}
    rotation {roll=-10}
    scale=<0.8 0.8 0.8>
    translation=<-2.61 -2.637 0.955>
   }
   n19 = "Transform (Modify) Bone - Bone: Dyn_MSTR_Jacket_Lower_r"(ComponentPose=@n18, Alpha=Get ADS_UP_R_Pistol(), Translation="0,2,0", Rotation="0,0,0", Scale="1,0.4,1") {
    rotationMode=BMM_Additive
    rotationSpace=BCS_BoneSpace
    scaleMode=BMM_Additive
    scaleSpace=BCS_BoneSpace
    translationMode=BMM_Additive
    translationSpace=BCS_BoneSpace
    alphaBoolBlend {blendOption=Linear}
    alphaScaleBiasClamp {bInterpResult=true}
    boneToModify {boneName=Dyn_MSTR_Jacket_Lower_r}
    scale {y=0.4}
    translation {y=2}
   }
   n20 = "Transform (Modify) Bone - Bone: Dyn_MSTR_Jacket_Upper_l"(ComponentPose=@n19, Alpha=Get ADS_UP_L_Assault() -> (ADS_UP_L_Assault:double), Translation="-1,0,0", Rotation="0,-15,0", Scale="1,1,1") {
    rotationMode=BMM_Additive
    rotationSpace=BCS_BoneSpace
    scaleMode=BMM_Replace
    scaleSpace=BCS_BoneSpace
    translationMode=BMM_Additive
    translationSpace=BCS_BoneSpace
    alphaBoolBlend {blendOption=Linear}
    alphaScaleBiasClamp {bInterpResult=true}
    boneToModify {boneName=Dyn_MSTR_Jacket_Upper_l}
    rotation {yaw=-15}
    translation {x=-1}
   }
   n21 = "Transform (Modify) Bone - Bone: Dyn_MSTR_Jacket_Upper_r"(ComponentPose=@n20, Alpha=Get ADS_UP_R_Assault() -> (ADS_UP_R_Assault:double), Translation="-2.5,0,0", Rotation="0,0,0", Scale="1,1,1") {
    rotationMode=BMM_Additive
    rotationSpace=BCS_BoneSpace
    scaleMode=BMM_Replace
    scaleSpace=BCS_BoneSpace
    translationMode=BMM_Additive
    translationSpace=BCS_BoneSpace
    alphaBoolBlend {blendOption=Linear}
    alphaScaleBiasClamp {bInterpResult=true}
    boneToModify {boneName=Dyn_MSTR_Jacket_Upper_r}
    translation {x=-2.5}
   }
   "Transform (Modify) Bone - Bone: Dyn_MSTR_Jacket_Lower_r"(ComponentPose=@n21, Alpha=Get ADS_UP_R_Assault(), Translation="0,2,0", Rotation="0,0,0", Scale="1,0.4,1") {
    rotationMode=BMM_Additive
    rotationSpace=BCS_BoneSpace
    scaleMode=BMM_Additive
    scaleSpace=BCS_BoneSpace
    translationMode=BMM_Additive
    translationSpace=BCS_BoneSpace
    alphaBoolBlend {blendOption=Linear}
    alphaScaleBiasClamp {bInterpResult=true}
    boneToModify {boneName=Dyn_MSTR_Jacket_Lower_r}
    scale {y=0.4}
    translation {y=2}
   }
   RigidBody!(ComponentPose, Alpha=0.8, OverrideWorldGravity:Vector="0,0,-980") -> (Pose) {
    bClampLinearTranslationLimitToRefPose=true
    bOverrideWorldGravity=true
    overridePhysicsAsset=&/M_MED_Calavera_Physics
    simulationSpace=BaseBoneSpace
    alphaBoolBlend {blendOption=Linear}
    baseBoneRef {boneName=pelvis}
    componentAppliedLinearAccClamp=<4500 4500 4500>
    overrideWorldGravity {z=-980}
    simSpaceSettings {dampingAlpha=0, maxAngularAcceleration=40, maxAngularVelocity=8, maxLinearAcceleration=2300, maxLinearVelocity=700, worldAlpha=0.8}
   }
   n22 = Control Rig!(Source=Copy Pose From Mesh() {bUseAttachedParent=true}, Alpha, IsEmote:bool=Get bIsPlayingEmote(), isDBNO:bool=Get bIsDBNO()) -> (Pose) {
    alphaBoolBlend {blendOption=Linear}
    controlRigAssetReference {blueprintRigClass=&/M_MED_Calavera_CtrlRig.M_MED_Calavera_CtrlRig_C}
   }
   n23 = RigidBody!(
    ComponentPose=Local To Component(LocalPose=@n22)
    Alpha
    bUseLocalLODThresholdOnly:bool=False
    OverrideWorldGravity="0,0,-980"
    ExternalForce:Vector
   ) -> (Pose) {
    bClampLinearTranslationLimitToRefPose=true
    overridePhysicsAsset=&/M_MED_Calavera_Tie_Physics
    simulationSpace=BaseBoneSpace
    alphaBoolBlend {blendOption=Linear}
    baseBoneRef {boneName=spine_05}
    componentAppliedLinearAccClamp=<5000 5000 5000>
    simSpaceSettings {maxAngularAcceleration=300, maxAngularVelocity=3, maxLinearAcceleration=500, maxLinearVelocity=350, worldAlpha=1, externalLinearDragV {x=4}}
   }
   n24 = RigidBody(ComponentPose=@n23, bUseLocalLODThresholdOnly=False, OverrideWorldGravity="0,0,-980") {
    bClampLinearTranslationLimitToRefPose=true
    overridePhysicsAsset=&M_MED_Calavera_Physics
    simulationSpace=BaseBoneSpace
    alphaBoolBlend {blendOption=Linear}
    baseBoneRef {boneName=pelvis}
    componentAppliedLinearAccClamp=<5000 5000 5000>
    simSpaceSettings {maxAngularAcceleration=300, maxAngularVelocity=3, maxLinearAcceleration=500, maxLinearVelocity=350, worldAlpha=1, externalLinearDragV {x=4}}
   }
   n25 = "Transform (Modify) Bone - Bone: attach_backpack"!(ComponentPose=@n24, bAlphaBoolEnabled:bool=True, Translation="0,3,5", Rotation="0,0,5", Scale="1,1,1") -> (Pose) {
    alphaInputType=Bool
    rotationMode=BMM_Additive
    rotationSpace=BCS_BoneSpace
    translationMode=BMM_Additive
    translationSpace=BCS_BoneSpace
    boneToModify {boneName=attach_backpack}
    rotation {roll=5}
    translation {y=3, z=5}
   }
   Output Pose(Result=Component To Local(ComponentPose=@n25)) {layerGroup=None}
  ]}
 EventGraph {nodes [
   K2Node_Event_0 = CustomEvent!() -> (OutputDelegate, then, DeltaTimeX:float=~) {is=title K2Node_CustomEvent}
   n1 = Gravity Override State Function DEPRECATED(
    execute=[@K2Node_Event_0.then, @n5.then]
    CharacterPartAnimInstance=Self-Reference()
    GameTimeInSeconds=GetGameTimeInSeconds()
    GravityOverrideParams=Get Gravity Override Params()
    RBAN Params=Get RBAN Params()
   )
   > Set Gravity Override(Gravity Override=@n1)
   n2 = Event OnCharacterCustomizationLoaded()
   > n3 = For Each Loop(Array:Array of Custom Character Part Object References=@n2.CustomParts) -> (Array Element:Custom Character Part Object Reference)
   > Branch(Condition="Equal (Enum)"(
     A:EFortCustomPartType Enum=GetPartType!(self:Custom Character Part Object Reference=@n3) -> (ReturnValue:EFortCustomPartType Enum=Head)
     B:EFortCustomPartType Enum=Backpack
    ))
   > n4 = Branch(Condition=Contains(SearchIn=GetObjectName(Object=@n3), Substring=RainbowSplash))
   > Set IsThisMyBackBling(IsThisMyBackBling:bool=true) -> (Output_Get:bool)
   Set IsThisMyBackBling(execute=@n4.else, IsThisMyBackBling=false)
   n5 = Event BlueprintUpdateAnimation()
  ]}
]
variables [
 ADS_UP_L_Assault {type=number}
 ADS_UP_L_Pistol {type=number}
 ADS_UP_R_Assault {type=number}
 ADS_UP_R_Pistol {type=number}
 Alt_weapon {type=boolean}
 Gravity Override {type=Vector}
 Gravity Override Params {type=GravityOverrideParamsStruct}
 "Integer >" {type="array<integer>"}
 Is In Front End {type=boolean}
 Is Rigid Body Disabled {type=boolean}
 IsThisMyBackBling {type=boolean}
 Left Leg {type=number}
 Left Leg_crouch {type=number}
 Left Leg_Shirt {type=number}
 RBAN Params {
  type=RBANParamsStruct
  default {
   dBNO {simSpaceSettings {
     maxAngularAcceleration=1000
     maxAngularVelocity=3
     maxLinearAcceleration=300
     maxLinearVelocity=700
     velocityScaleZ=0
     externalLinearDragV=<0 0.5 0>
    }}
   dBNO_Carried {simSpaceSettings {
     maxAngularAcceleration=1000
     maxAngularVelocity=3
     maxLinearAcceleration=300
     maxLinearVelocity=700
     velocityScaleZ=0
     externalLinearDragV=<0 0.5 0>
    }}
   emoteOrMelee {simSpaceSettings {
     maxAngularAcceleration=1000
     maxAngularVelocity=3
     maxLinearAcceleration=300
     maxLinearVelocity=700
     velocityScaleZ=0
     externalLinearDragV=<0 0.5 0>
    }}
   front_End {simSpaceSettings {
     maxAngularAcceleration=1000
     maxAngularVelocity=3
     maxLinearAcceleration=300
     maxLinearVelocity=700
     velocityScaleZ=0
     externalLinearDragV=<0 0.5 0>
    }}
   onGround_Crouch_Moving {simSpaceSettings {
     maxAngularAcceleration=1000
     maxAngularVelocity=3
     maxLinearAcceleration=300
     maxLinearVelocity=700
     velocityScaleZ=0
     externalLinearDragV=<0 0.5 0>
    }}
   onGround_Falling {simSpaceSettings {
     maxAngularAcceleration=1000
     maxAngularVelocity=3
     maxLinearAcceleration=300
     maxLinearVelocity=700
     velocityScaleZ=0
     externalLinearDragV=<0 0.5 0>
    }}
   onGround_Moving {simSpaceSettings {
     maxAngularAcceleration=1000
     maxAngularVelocity=3
     maxLinearAcceleration=300
     maxLinearVelocity=700
     velocityScaleZ=0
     externalLinearDragV=<0 0.5 0>
    }}
   onGround_Standing {simSpaceSettings {
     maxAngularAcceleration=1000
     maxAngularVelocity=3
     maxLinearAcceleration=300
     maxLinearVelocity=700
     velocityScaleZ=0
     externalLinearDragV=<0 0.5 0>
    }}
   rail_Grinding {simSpaceSettings {
     maxAngularAcceleration=1000
     maxAngularVelocity=3
     maxLinearAcceleration=300
     maxLinearVelocity=700
     velocityScaleZ=0
     worldAlpha=0.1
     externalLinearDragV=<0 0.5 0>
    }}
   riding_Moving {simSpaceSettings {
     maxAngularAcceleration=1000
     maxAngularVelocity=3
     maxLinearAcceleration=300
     maxLinearVelocity=700
     velocityScaleZ=0
     externalLinearDragV=<0 0.5 0>
    }}
   riding_Moving_Targeting {simSpaceSettings {
     maxAngularAcceleration=1000
     maxAngularVelocity=3
     maxLinearAcceleration=300
     maxLinearVelocity=700
     velocityScaleZ=0
     externalLinearDragV=<0 0.5 0>
    }}
   riding_Standing {simSpaceSettings {
     maxAngularAcceleration=1000
     maxAngularVelocity=3
     maxLinearAcceleration=300
     maxLinearVelocity=700
     velocityScaleZ=0
     externalLinearDragV=<0 0.5 0>
    }}
   skydive_DivingDown {simSpaceSettings {
     maxAngularAcceleration=1000
     maxAngularVelocity=3
     maxLinearAcceleration=300
     maxLinearVelocity=700
     velocityScaleZ=0
     externalLinearDragV=<0 0.5 0>
    }}
   skydive_FloatingDown {
    componentLinearAccScale=<0 0.5 0>
    componentLinearVelScale=<0 0.5 0>
    simSpaceSettings {
     maxAngularAcceleration=1000
     maxAngularVelocity=3
     maxLinearAcceleration=300
     maxLinearVelocity=700
     velocityScaleZ=0
     externalLinearDragV=<0 0.5 0>
    }
   }
   skydive_GoingUp {simSpaceSettings {
     maxAngularAcceleration=1000
     maxAngularVelocity=3
     maxLinearAcceleration=300
     maxLinearVelocity=700
     velocityScaleZ=0
     externalLinearDragV=<0 0.5 0>
    }}
   skydive_Parachuting {
    componentLinearAccScale=<0 0.1 0.3>
    componentLinearVelScale=<0 0.1 0.3>
    simSpaceSettings {
     maxAngularAcceleration=1000
     maxAngularVelocity=3
     maxLinearAcceleration=300
     maxLinearVelocity=700
     velocityScaleZ=0
     externalLinearDragV=<0 0.5 0>
    }
   }
   skydive_Surfing {simSpaceSettings {
     maxAngularAcceleration=1000
     maxAngularVelocity=3
     maxLinearAcceleration=300
     maxLinearVelocity=700
     velocityScaleZ=0
     externalLinearDragV=<0 0.5 0>
    }}
   swimming {simSpaceSettings {
     maxAngularAcceleration=1000
     maxAngularVelocity=3
     maxLinearAcceleration=300
     maxLinearVelocity=700
     velocityScaleZ=0
     externalLinearDragV=<0 0.5 0>
    }}
   tactical_Sprint {simSpaceSettings {
     maxAngularAcceleration=1000
     maxAngularVelocity=3
     maxLinearAcceleration=300
     maxLinearVelocity=700
     velocityScaleZ=0
     externalLinearDragV=<0 0.5 0>
    }}
   ziplining {simSpaceSettings {
     maxAngularAcceleration=1000
     maxAngularVelocity=3
     maxLinearAcceleration=300
     maxLinearVelocity=700
     velocityScaleZ=0
     externalLinearDragV=<0 0.5 0>
    }}
  }
 }
 Right Leg {type=number}
 Right Leg_Crouch {type=number}
 Skel Mesh {type=SkeletalMeshComponent}
 Standing_Moving {type=boolean}
]
