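# Physics asset for the face accessory part of the medium male Calavera player body, defining physics bodies for its hair bones.
# Small capsules form three separate chains: a three-link main strand below Dyn_Master_Hair, plus left and right curls, each one link below its own master bone.
# All joints share the same angular limits, with asymmetric swing and almost no twist.
# No constraints connect the three chains to each other.
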PhysicsAsset /BRCosmetics/Characters/Player/Male/Medium/Bodies/M_MED_Calavera/Meshes/Parts/M_MED_Calavera_FaceAcc_Physics
bodies {
 Dyn_Curl_Hair_l1 [{shapeType=Capsule, length=1, radius=0.5, shapeName=Dyn_Curl_Hair_l1_capsule, center=<0.5 0 0>, rotation=<rot 90 0 0>}]
 Dyn_Curl_Hair_r1 [{shapeType=Capsule, length=1, radius=0.5, shapeName=Dyn_Curl_Hair_r1_capsule, center=<0.5 0 0>, rotation=<rot 90 0 0>}]
 Dyn_Hair1 [{shapeType=Capsule, length=0.5, radius=0.5, shapeName=Dyn_Hair1_capsule, center=<0.5 0 0>, rotation=<rot 90 0 0>}]
 Dyn_Hair2 [{shapeType=Capsule, length=0.5, radius=0.5, shapeName=Dyn_Hair2_capsule, center=<0.5 0 0>, rotation=<rot 90 0 0>}]
 Dyn_Hair3 [{shapeType=Capsule, length=0.5, radius=0.5, shapeName=Dyn_Hair3_capsule, center=<0.5 0 0>, rotation=<rot 90 0 0>}]
 Dyn_Master_Curl_Hair_l [{shapeType=Capsule, length=0.5, radius=0.5, shapeName=Dyn_Master_Curl_Hair_l_capsule}]
 Dyn_Master_Curl_Hair_r1 [{shapeType=Capsule, length=0.5, radius=0.5, shapeName=Dyn_Master_Curl_Hair_r1_capsule}]
 Dyn_Master_Hair [{shapeType=Capsule, length=1, radius=1, shapeName=Dyn_Master_Hair_capsule}]
}
constraints [
 {bone1Name=Dyn_Hair1, bone2Name=Dyn_Master_Hair, swing1LimitDegrees=20, swing2LimitDegrees=50, twistLimitDegrees=1}
 {bone1Name=Dyn_Hair2, bone2Name=Dyn_Hair1, swing1LimitDegrees=20, swing2LimitDegrees=50, twistLimitDegrees=1}
 {bone1Name=Dyn_Hair3, bone2Name=Dyn_Hair2, swing1LimitDegrees=20, swing2LimitDegrees=50, twistLimitDegrees=1}
 {bone1Name=Dyn_Curl_Hair_l1, bone2Name=Dyn_Master_Curl_Hair_l, swing1LimitDegrees=20, swing2LimitDegrees=50, twistLimitDegrees=1}
 {bone1Name=Dyn_Curl_Hair_r1, bone2Name=Dyn_Master_Curl_Hair_r1, swing1LimitDegrees=20, swing2LimitDegrees=50, twistLimitDegrees=1}
]
